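# Idle effect for the Lucky Rider pickaxe that puts persistent warm glows and lights on the weapon's sockets.
# A single emitter bursts two particles that never expire by lifetime and follow sockets each update through a system socket reader indexed by execution index.
# Each particle feeds two sprite renderers, a light and a mesh renderer, with camera offset and separate attributes for light, cone and air glow.
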NiagaraSystem /BRCosmetics/Weapons/FORT_Melee/Pickaxe_Lucky_Rider/FX/NS_Pickaxe_LuckyRider_Idle
dependencies {
 usedDataInterfaces [&NiagaraDataInterfaceSocketReader]
 usedDynamicInputs []
 usedModules [
  &self:SystemSpawnScript.SystemScriptSource.SystemScriptGraph.NiagaraNodeAssignment_0.SetVariables_#1
  &SystemState
 ]
}
emitters {Minimal {
  emitterUpdateScript [
   &EmitterState(Life Cycle Mode=::NewEnumerator1, Loop Behavior=::NewEnumerator1, Loop Duration Mode=::NewEnumerator1, UseLoopDelay=false)
   &SpawnBurst_Instantaneous(Spawn Count=2, Use Spawn Probability=false, UseLoopCountLimit=false)
  ]
  particleSpawnScript [&InitializeParticle(
    Color=<rgba 1 0.8462 0.4329 0.4>
    Mesh Scale=<0.9 1.2 0.9>
    Mesh Scale Mode=::NewEnumerator4
    Sprite Size Mode=::NewEnumerator3
    Uniform Sprite Size=30
    UsePositionOffset=false
    Write Material Random=true
   )]
  particleUpdateScript [
   &ParticleState(Kill Particles When Lifetime Has Elapsed=false)
   &SocketLocation(
    SamplePosOffset=false
    Socket Index=&ReturnExecIndex()
    Socket Reader=$System.Socket Reader
    Socket Sampling Mode=::NewEnumerator3
    VelocitySpeedLimit=false
   )
   &FN_Light_Attributes(
    Color=$Particles.Color
    Create Color Attribute=true
    Create Exponent Attribute=true
    Create Position Attribute=true
    Create Radius Attribute=true
    Exponent=2
    Position=&AddVectorToPosition(Position=$Particles.Position, Vector=<0 20 0>)
    Radius=30
    Scale Color=3
   )
   &CameraOffset(Camera Offset Amount=5)
   /SetVariables(
    Particles.ConeLightColor:LinearColor=<rgba 1 0.8692 0.4187 0.2>
    Particles.ConePosition:NiagaraPosition=&AddVectorToPosition(Position=$Particles.Position, Vector=<0 -2 0>)
   ) {isSetParameters=true}
   /SetVariables(
    Particles.AirGlowDynamicParam:Vector4f=<4 0 0 0>
    Particles.AirGlowSpriteSize:Vector2f=&Vector2DFromFloat(Value=7.5)
   ) {isSetParameters=true}
  ]
  rendererClasses [
   &NiagaraSpriteRendererProperties
   &NiagaraLightRendererProperties
   &NiagaraMeshRendererProperties
  ]
  renderers [
   {rendererClass=&NiagaraSpriteRendererProperties}
   {rendererClass=&NiagaraLightRendererProperties, rendererIndex=1}
   {rendererClass=&NiagaraMeshRendererProperties, rendererIndex=2}
   {rendererClass=&NiagaraSpriteRendererProperties, rendererIndex=3}
  ]
 }}
summary {
 systemName=NS_Pickaxe_LuckyRider_Idle
 emitters [{
   emitterName=Minimal
   rendererClasses [
    &NiagaraSpriteRendererProperties
    &NiagaraLightRendererProperties
    &NiagaraMeshRendererProperties
   ]
  }]
 userVariables []
}
